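# Debug Niagara system in an Old folder for the FareSporeCookie emote, layering blue glows with drifting mesh particles.
# FlareBurst and Glow draw layered sprite flares with time-of-day-scaled lights; Glow samples the parent skeleton.
# Square_far001 spawns rising, curl-noise meshes, tinted separately for front end and gameplay.
# Disabled emitters hold a light and a treadmilling mesh-grid visualizer whose column heights react to nearest body joint, collision and audio.
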NiagaraSystem /BRCosmetics/Animation/Game/MainPlayer/Emotes/FareSporeCookie/FX/Old/NS_FareSporeCookie_Debug
dependencies {
 usedDataInterfaces [
  &NiagaraDataInterfaceArrayFloat3
  &NiagaraDataInterfaceArrayFloat
  &NiagaraDataInterfaceCurve
  &NiagaraDataInterfaceCollisionQuery
  &NiagaraDataInterfaceDebugDraw
  &NiagaraDataInterfaceVectorCurve
 ]
 usedDynamicInputs [&SimulationPosition]
 usedModules [&SystemState]
}
emitters {
 BlowingParticles {
  bEnabled=false
  emitterUpdateScript [
   &EmitterState(Life Cycle Mode=::NewEnumerator1, UseLoopDelay=false)
   &SpawnBurst_Instantaneous(Use Spawn Probability=false, UseLoopCountLimit=false)
  ]
  particleSpawnScript [&InitializeParticle(
    Color=<rgba 1.276 3.8972 5 1>
    Mass Max=1.5
    Mass Mode=::NewEnumerator3
    Sprite Rotation Mode=::NewEnumerator2
    Sprite Size Mode=::NewEnumerator1
    Uniform Sprite Size Max=180
    Uniform Sprite Size Min=150
    UsePositionOffset=false
    Write Material Random=true
   )]
  particleUpdateScript [
   &ParticleState(Kill Particles When Lifetime Has Elapsed=false, Let Infinitely Lived Particles Die When Emitter Deactivates=true)
   &FN_Light_Attributes(
    Color=<rgba 0 0.3024 1 1>
    Create Color Attribute=true
    Create Exponent Attribute=true
    Create Radius Attribute=true
    Exponent=1
    Radius=150
    Scale Color=10
   )
   /SetVariables() {isSetParameters=true}
  ]
  rendererClasses [&NiagaraLightRendererProperties]
  renderers [{rendererClass=&NiagaraLightRendererProperties, rendererIndex=0}]
 }
 FlareBurst {
  emitterUpdateScript [
   &EmitterState(Life Cycle Mode=::NewEnumerator1, Loop Behavior=::NewEnumerator1, Loop Duration Mode=::NewEnumerator1, UseLoopDelay=false)
   &SpawnRate(SpawnRate=2, Use Spawn Probability=false)
   &SpawnBurst_Instantaneous(Spawn Time=0.01, Use Spawn Probability=false, UseLoopCountLimit=false)
  ]
  particleSpawnScript [
   &InitializeParticle(
    Color=<rgba 0.0019 0.0436 1 1>
    Mass Max=2
    Mass Mode=::NewEnumerator3
    Position Offset=<0 0 80>
    Position Offset Coordinate Space=::Local
    Sprite Rotation Mode=::NewEnumerator2
    Sprite Size Mode=::NewEnumerator3
    Uniform Sprite Size=&RandomRangeFloat(Minimum=150, Maximum=300)
    UsePositionOffset=true
    Write Material Random=true
   )
   &CameraOffset(Camera Offset Amount=30)
  ]
  particleUpdateScript [
   &ParticleState()
   &CameraOffset(Camera Offset Amount=50) {module=CameraOffset001}
   &SampleSkeletalMesh(
    AdditionalYawPitchRoll=false
    Bone Sampling Mode:ENiagaraBoneSamplingMode=::NewEnumerator4
    Skeletal Mesh:NiagaraDataInterfaceSkeletalMesh=$User.SkeletalMesh
   ) {enabled=false, module=SampleSkeletalMesh001}
   &ScaleColor(
    Scale Alpha=&FloatFromCurve(
     FloatCurve=DataInterface() {propertyValues {Curve {keys [
         <at 0 0 RCIM_Cubic>
         {arriveTangent=0.1895, interpMode=RCIM_Cubic, leaveTangent=0.1895, tangentMode=RCTM_User, time=0.1, value=1}
         {arriveTangent=-0.0202, interpMode=RCIM_Cubic, leaveTangent=-0.0202, tangentMode=RCTM_User, time=0.7, value=1}
         <at 1.0011 0.002 RCIM_Cubic>
        ]}}}
     Scale Curve=10
    )
    ScaleA=true
    ScaleRGB=true
   )
   &ScaleSpriteSize(Scale Factor=&Vector2DFromFloat(Value=&FloatFromCurve(FloatCurve=DataInterface() {propertyValues {Curve {keys [
          {interpMode=RCIM_Cubic, tangentMode=RCTM_User, time=0, value=0}
          {arriveTangent=-0.0009, interpMode=RCIM_Cubic, leaveTangent=-0.0009, tangentMode=RCTM_User, time=0.2, value=1}
          {tangentMode=RCTM_User, time=1, value=0}
         ]}}})))
   /SetVariables(
    Particles.StarColor:LinearColor=&MultiplyLinearColor(Linear Color=$Particles.Color, Scale Factor=&MakeVector4_FromLinearColor(LinearColor=<rgba 1 1 1 1>))
    Particles.StarRotation:NiagaraFloat=&RandomRangeFloat(Minimum=0, Maximum=360)
    Particles.StarSize:Vector2f=&Multiply_Vector2(A=$Particles.SpriteSize, B=<2.75 2.75>)
   ) {isSetParameters=true, module=SetVariables_#1}
   /SetVariables(
    Particles.StreaksColor:LinearColor=&MultiplyLinearColor(
     Linear Color=$Particles.Color
     Scale Factor=&MakeVector4_FromLinearColor(LinearColor=<rgba 0.6 0.8 1 0.01>)
    )
    Particles.StreaksRotation:NiagaraFloat=&RandomRangeFloat(Minimum=0, Maximum=360)
    Particles.StreaksSize:Vector2f=&Multiply_Vector2(A=$Particles.SpriteSize, B=<7 7>)
   ) {isSetParameters=true, module=SetVariables_#2}
   /SetVariables(
    Particles.AnamorphicColor:LinearColor=&MultiplyLinearColor(
     Linear Color=$Particles.Color
     Scale Factor=&MakeVector4_FromLinearColor(LinearColor=<rgba 0.8 0.9 1 0.01>)
    )
    Particles.AnamorphicRotation=90
    Particles.AnamorphicSize:Vector2f=&Multiply_Vector2(A=$Particles.SpriteSize, B=<3 9>)
   ) {isSetParameters=true, module=SetVariables_#3}
   /SetVariables(
    Particles.HaloColor:LinearColor=&MultiplyLinearColor(Linear Color=$Particles.Color, Scale Factor=&MakeVector4_FromLinearColor(LinearColor=<rgba 1 1 1 0.025>))
    Particles.HaloSize:Vector2f=&Multiply_Vector2(A=$Particles.SpriteSize, B=<5 5>)
   ) {isSetParameters=true, module=SetVariables_#4}
   /SetVariables(
    Particles.HaloColor_02:LinearColor=&MultiplyLinearColor(Linear Color=$Particles.Color, Scale Factor=&MakeVector4_FromLinearColor(LinearColor=<rgba 1 1 1 0.05>))
    Particles.HaloSize_02:Vector2f=&Multiply_Vector2(A=$Particles.SpriteSize, B=<1.5 1.5>)
   ) {isSetParameters=true, module=SetVariables_#5}
   /SetVariables(
    Particles.AirGlowColor:LinearColor=&MultiplyLinearColor(Linear Color=$Particles.Color, Scale Factor=&MakeVector4_FromLinearColor(LinearColor=<rgba 1 1 1 0.005>))
    Particles.AirGlowSize:Vector2f=&Multiply_Vector2(A=$Particles.SpriteSize, B=<4 4>)
   ) {isSetParameters=true, module=SetVariables_#6}
   &FN_Light_Attributes(
    Color=<rgba 1 1 1 0.2>
    Create Color Attribute=true
    Create Exponent Attribute=true
    Create Position Attribute=true
    Create Radius Attribute=true
    Enable Particle Light Enable=true
    Light Enabled=true
    Position=/Niagara/DynamicInputs/Lerp/Lerp_Position(
     A:NiagaraPosition=$Particles.SampleSkeletalMesh001.SampledPosition
     B:NiagaraPosition=/Niagara/DynamicInputs/Helpers/SimulationPosition()
     Alpha:NiagaraFloat=0.5
    )
    Radius=300
    Scale Color=&FloatFromCurve(
     FloatCurve=DataInterface() {propertyValues {Curve {keys [
         {arriveTangent=9.0725, interpMode=RCIM_Cubic, leaveTangent=9.0725, tangentMode=RCTM_User, time=0, value=0}
         {arriveTangent=-2.5249, interpMode=RCIM_Cubic, leaveTangent=-2.5249, tangentMode=RCTM_User, time=0.1408, value=0.9847}
         {arriveTangent=-0.0405, interpMode=RCIM_Cubic, leaveTangent=-0.0405, tangentMode=RCTM_User, time=1, value=0}
        ]}}}
     Scale Curve=10
    )
   )
   /SetVariables(Particles.TempPosition:Vector3f=$Particles.FN_Light_Attributes.Position) {isSetParameters=true, module=SetVariables_#7}
   &FN_Gen5_Light(
    Attenutation=600
    Exponent=8
    Intensity:NiagaraFloat=&Multiply_Float(A=2, B=&FNTimeOfDayScalar(Night Scale=2, Morning Scale=6, Day Scale=8, Evening Scale=5))
   )
  ]
  rendererClasses [&NiagaraSpriteRendererProperties]
  renderers [
   {rendererClass=&NiagaraSpriteRendererProperties}
   {rendererClass=&NiagaraSpriteRendererProperties, rendererIndex=1}
   {rendererClass=&NiagaraSpriteRendererProperties, rendererIndex=2}
   {rendererClass=&NiagaraSpriteRendererProperties, rendererIndex=3}
  ]
 }
 Glow {
  emitterUpdateScript [
   &EmitterState(Life Cycle Mode=::NewEnumerator1, Loop Behavior=::NewEnumerator1, Loop Delay=0.1, Loop Duration Mode=::NewEnumerator1, UseLoopDelay=true)
   &SpawnRate(SpawnRate=2, Use Spawn Probability=false)
   &SpawnBurst_Instantaneous(Spawn Time=0.01, Use Spawn Probability=false, UseLoopCountLimit=false)
  ]
  particleSpawnScript [
   &InitializeParticle(
    Color=<rgba 0.0626 0.1202 1 1>
    Mass Max=2
    Mass Mode=::NewEnumerator3
    Position Offset Coordinate Space=::Local
    Sprite Rotation Mode=::NewEnumerator2
    Sprite Size Mode=::NewEnumerator3
    Uniform Sprite Size=&RandomRangeFloat(Minimum=150, Maximum=250)
    UsePositionOffset=true
    Write Material Random=true
   )
   &CameraOffset(Camera Offset Amount=-100)
  ]
  particleUpdateScript [
   &ParticleState()
   &SkeletalMeshLocation(
    Bone Sampling Mode=::NewEnumerator4
    MaskSpawnGroup=false
    Orientation Sampling=::NewEnumerator1
    Sampled Position Offset Coordinate Space:ENiagaraSkelMeshTransforms=::NewEnumerator0
    Sampled Scale Coordinate Space:ENiagaraCoordinateSpace=::Simulation
    Skeletal Mesh=DataInterface() {propertyValues {
      DefaultMesh=null
      ExcludeBoneName=root
      OptionalComponentClass=null
      PreviewMesh=/Game/Characters/Player/Male/Small/Bodies/M_Sml_Ninja_04/Meshes/SK_M_SML_Ninja_04_Body
      SkinningMode=SkinOnTheFly
      SoftSourceActor=null
      SourceMode=AttachParent
      UvSetIndex=0
      WholeMeshLOD=-1
      bExcludeBone=false
      bReadDeformedGeometry=true
      bRequireCurrentFrameData=true
      ComponentTags []
      FilteredBones [pelvis,neck_02]
      FilteredSockets []
      MeshUserParameter {parameter {name=None, typeDefHandle {registeredTypeIndex=1}, varData []}}
      SamplingRegions []
     }}
   )
   &ScaleColor(
    Scale Alpha=&FloatFromCurve(
     FloatCurve=DataInterface() {propertyValues {Curve {keys [
         <at 0 0 RCIM_Cubic>
         {arriveTangent=0.1895, interpMode=RCIM_Cubic, leaveTangent=0.1895, tangentMode=RCTM_User, time=0.1, value=1}
         {arriveTangent=-0.0202, interpMode=RCIM_Cubic, leaveTangent=-0.0202, tangentMode=RCTM_User, time=0.7, value=1}
         <at 1.0011 0.002 RCIM_Cubic>
        ]}}}
     Scale Curve=10
    )
    ScaleA=true
    ScaleRGB=true
   )
   &ScaleSpriteSize(Scale Factor=&Vector2DFromFloat(Value=&FloatFromCurve(FloatCurve=DataInterface() {propertyValues {Curve {keys [
          {interpMode=RCIM_Cubic, tangentMode=RCTM_User, time=0, value=0}
          {arriveTangent=-0.0009, interpMode=RCIM_Cubic, leaveTangent=-0.0009, tangentMode=RCTM_User, time=0.2, value=1}
          {tangentMode=RCTM_User, time=1, value=0}
         ]}}})))
   /SetVariables(
    Particles.StarColor:LinearColor=&MultiplyLinearColor(Linear Color=$Particles.Color, Scale Factor=&MakeVector4_FromLinearColor(LinearColor=<rgba 1 1 1 1>))
    Particles.StarRotation:NiagaraFloat=&RandomRangeFloat(Minimum=0, Maximum=360)
    Particles.StarSize:Vector2f=&Multiply_Vector2(A=$Particles.SpriteSize, B=<2.75 2.75>)
   ) {isSetParameters=true, module=SetVariables_#1}
   /SetVariables(
    Particles.StreaksColor:LinearColor=&MultiplyLinearColor(
     Linear Color=$Particles.Color
     Scale Factor=&MakeVector4_FromLinearColor(LinearColor=<rgba 0.6 0.8 1 0.01>)
    )
    Particles.StreaksRotation:NiagaraFloat=&RandomRangeFloat(Minimum=0, Maximum=360)
    Particles.StreaksSize:Vector2f=&Multiply_Vector2(A=$Particles.SpriteSize, B=<7 7>)
   ) {isSetParameters=true, module=SetVariables_#2}
   /SetVariables(
    Particles.AnamorphicColor:LinearColor=&MultiplyLinearColor(
     Linear Color=$Particles.Color
     Scale Factor=&MakeVector4_FromLinearColor(LinearColor=<rgba 0.8 0.9 1 0.01>)
    )
    Particles.AnamorphicRotation=90
    Particles.AnamorphicSize:Vector2f=&Multiply_Vector2(A=$Particles.SpriteSize, B=<3 9>)
   ) {isSetParameters=true, module=SetVariables_#3}
   /SetVariables(
    Particles.HaloColor:LinearColor=&MultiplyLinearColor(Linear Color=$Particles.Color, Scale Factor=&MakeVector4_FromLinearColor(LinearColor=<rgba 1 1 1 0.025>))
    Particles.HaloSize:Vector2f=&Multiply_Vector2(A=$Particles.SpriteSize, B=<5 5>)
   ) {isSetParameters=true, module=SetVariables_#4}
   /SetVariables(
    Particles.HaloColor_02:LinearColor=&MultiplyLinearColor(Linear Color=$Particles.Color, Scale Factor=&MakeVector4_FromLinearColor(LinearColor=<rgba 1 1 1 0.05>))
    Particles.HaloSize_02:Vector2f=&Multiply_Vector2(A=$Particles.SpriteSize, B=<1.5 1.5>)
   ) {isSetParameters=true, module=SetVariables_#5}
   /SetVariables(
    Particles.AirGlowColor:LinearColor=&MultiplyLinearColor(
     Linear Color=$Particles.Color
     Scale Factor=&MakeVector4_FromLinearColor(LinearColor=&MultiplyLinearColor(
       Linear Color=&FNTimeOfDayColor(Night Color=<rgba 1 1 1 0.8>, Morning Color=<rgba 1 1 1 0.5>, Day Color=<rgba 1 1 1 1>, Evening Color=<rgba 1 1 1 0.5>)
       Scale Factor=&MakeVector4_FromLinearColor(LinearColor=<rgba 1 1 1 0.005>)
      ))
    )
    Particles.AirGlowSize:Vector2f=&Multiply_Vector2(A=$Particles.SpriteSize, B=<4 4>)
   ) {isSetParameters=true, module=SetVariables_#6}
   &FN_Light_Attributes(
    Color=<rgba 1 1 1 0.2>
    Create Color Attribute=true
    Create Exponent Attribute=true
    Create Position Attribute=true
    Create Radius Attribute=true
    Enable Particle Light Enable=true
    Light Enabled=true
    Position=&Lerp_Position(
     A:NiagaraPosition=$Particles.SampleSkeletalMesh001.SampledPosition
     B:NiagaraPosition=&SimulationPosition()
     Alpha:NiagaraFloat=0.5
    )
    Radius=300
    Scale Color=&FloatFromCurve(
     FloatCurve=DataInterface() {propertyValues {Curve {keys [
         {arriveTangent=9.0725, interpMode=RCIM_Cubic, leaveTangent=9.0725, tangentMode=RCTM_User, time=0, value=0}
         {arriveTangent=-2.5249, interpMode=RCIM_Cubic, leaveTangent=-2.5249, tangentMode=RCTM_User, time=0.1408, value=0.9847}
         {arriveTangent=-0.0405, interpMode=RCIM_Cubic, leaveTangent=-0.0405, tangentMode=RCTM_User, time=1, value=0}
        ]}}}
     Scale Curve=10
    )
   )
   /SetVariables(Particles.TempPosition:Vector3f=$Particles.FN_Light_Attributes.Position) {isSetParameters=true, module=SetVariables_#7}
   &FN_Gen5_Light(
    Attenutation=600
    Exponent=8
    Intensity:NiagaraFloat=&Multiply_Float(A=2, B=&FNTimeOfDayScalar(Night Scale=2, Morning Scale=6, Day Scale=8, Evening Scale=5))
   )
  ]
  rendererClasses [&NiagaraSpriteRendererProperties]
  renderers [
   {rendererClass=&NiagaraSpriteRendererProperties}
   {rendererClass=&NiagaraSpriteRendererProperties, rendererIndex=1}
   {rendererClass=&NiagaraSpriteRendererProperties, rendererIndex=2}
   {rendererClass=&NiagaraSpriteRendererProperties, rendererIndex=3}
  ]
 }
 Square_far001 {
  emitterUpdateScript [
   &EmitterState(
    Inactive Response=::NewEnumerator1
    Life Cycle Mode=::NewEnumerator1
    Loop Behavior=::NewEnumerator1
    Loop Duration Mode=::NewEnumerator1
    UseLoopDelay=false
   )
   &SpawnRate(
    SpawnRate=&FloatFromCurve(
     FloatCurve=DataInterface() {propertyValues {Curve {keys [<at 0 1>,<at 1 0>]}}}
     CurveIndex=$Emitter.NormalizedLoopAge
     Scale Curve=&Lerp_Float(A=0, B=100, Alpha=&ClampFloat(Float=1))
    )
    Use Spawn Probability=false
   )
  ]
  particleSpawnScript [
   &PartitionParticles()
   &InitializeParticle(
    Color=&RandomRangeLinearColor(Minimum=<rgba 0 2.9765 15 1>, Maximum=<rgba 0.5496 0.0802 3.8509 1>)
    Mesh Scale=&Multiply_Vector(
     A=<0.7 0.7 0.7>
     B=&VectorFromFloat(Value=&UniformRangedFloat(
       Minimum=&Multiply_Float(
        A=0.1
        B=&FloatFromCurve(
         FloatCurve=DataInterface() {propertyValues {Curve {keys [<at 0 1>,<at 1 1>]}}}
         CurveIndex=&Divide_Float(A=&VectorLength(Vector=$Engine.Owner.Velocity), B=9000)
        )
       )
       Maximum=&Multiply_Float(
        A=0.2
        B=&FloatFromCurve(
         FloatCurve=DataInterface() {propertyValues {Curve {keys [<at 0 1>,<at 1 1.2>]}}}
         CurveIndex=&Divide_Float(A=&VectorLength(Vector=$Engine.Owner.Velocity), B=9000)
        )
       )
      ))
    )
    Mesh Scale Mode=::NewEnumerator4
    UsePositionOffset=false
    Write Material Random=true
   )
   &ShapeLocation(
    Box Size=<125 280 10>
    MaskSpawnGroup=false
    Override Seed=false
    Shape Origin=$Particles.SkeletalMeshLocation001.SampledPosition
    Shape Primitive=::NewEnumerator2
    Surface Only=false
    UseOwnerScale=true
   )
   &RegionMask(
    Invert Region=true
    Offset:Vector3f=<0 10 0>
    Offset Space:ENiagaraCoordinateSpace=::Local
    Outer Float=0
    Region Dimensions:Vector3f=<225 500 1500>
    Region Falloff=300
    Region Type:ENiagaraDebugDrawShape=::NewEnumerator1
    Rotation:Vector3f=$Engine.Owner.SystemYAxis
    Rotation Angle Type:ENiagara_AngleInput=::NewEnumerator0
    UseOffset=true
    UseRotation=true
   )
   &AddVelocity(Coordinate Space=::World, Velocity=<0 0 100>)
  ]
  particleUpdateScript [
   &ParticleState()
   &ScaleMeshSize(Scale Factor=&VectorFromFloat(Value=&MaskFloat(
      Mask Source=1
      Mask Parameter=$Particles.PartitionParticles.Partition
      Mask=0
      Masked Float=&Lerp_Float(A=1, B=0, Alpha=&ClampFloat(Float=$User.AltTimeline))
      Passthrough Float=1
     )))
   &ScaleMeshSize(Scale Factor=&VectorFromCurve(VectorCurve=DataInterface() {propertyValues {XCurve {keys [<at 0 1>,<at 1 0>]}, YCurve {keys [<at 0 1>,<at 1 0>]}, ZCurve {keys [<at 0 1>,<at 1 0>]}}})) {module=ScaleMeshSize001}
   &Drag(Drag=2, Use Linear Drag=true, Use Rotational Drag=false)
   &CurlNoiseForce(Bias Noise=false, Noise Strength=200, Pan Noise=false)
   &SolveForcesAndVelocity(Clamp Velocity=false, Limit Acceleration=false)
   &ScaleColor(
    Scale RGB=&FN_FrontEndVector3Modifier(
     Gameplay=&MakeVectorFromLinearColorRGB(Color=<rgba 0.5168 0.5579 0.7 1>)
     Front End=&MakeVectorFromLinearColorRGB(Color=$User.FrontEnd_color)
    )
    ScaleA=true
    ScaleRGB=true
   ) {module=ScaleColor002}
   &ScaleMeshSize(Scale Factor=&VectorFromFloat(Value=&FloatFromCurve(
      FloatCurve=DataInterface() {propertyValues {Curve {keys [
          {arriveTangent=-1.3985, interpMode=RCIM_Cubic, leaveTangent=-1.3985, tangentMode=RCTM_User, time=0, value=1}
          {arriveTangent=-0.5, interpMode=RCIM_Cubic, leaveTangent=-0.5, time=0.1, value=0.6}
          {arriveTangent=-0.5, interpMode=RCIM_Cubic, leaveTangent=-0.5, time=0.3, value=0.85}
          {arriveTangent=-0.1, interpMode=RCIM_Cubic, leaveTangent=-0.1, time=0.5, value=0.4}
          {arriveTangent=-0.375, interpMode=RCIM_Cubic, leaveTangent=-0.375, time=0.8, value=0.8}
          {arriveTangent=1, interpMode=RCIM_Cubic, leaveTangent=1, time=0.9, value=0.25}
          <at 1 1>
         ]}}}
      CurveIndex=&Add_Float(A=$Particles.NormalizedAge, B=&RandomRangeFloat(Minimum=0, Maximum=0.5))
      Scale Curve=1.5
     ))) {module=ScaleMeshSize002}
   &CameraOffset()
  ]
  rendererClasses [&NiagaraMeshRendererProperties]
  renderers [{rendererClass=&NiagaraMeshRendererProperties, rendererIndex=0}]
 }
 Visualizer {
  bEnabled=false
  emitterUpdateScript [
   &EmitterState(
    Inactive Response=::NewEnumerator1
    Life Cycle Mode=::NewEnumerator1
    Loop Behavior=::NewEnumerator1
    Loop Duration Mode=::NewEnumerator1
    UseLoopDelay=false
   )
   /SetVariables(
    Emitter.OriginOffset:Vector3f=&TransformVector(Vector=<0 0 0>, Destination Space=::World)
    Emitter.PaddingPerCell=10
    Emitter.VolumeRadius=200
   ) {isSetParameters=true}
   /Niagara/Modules/Spawn/Location/SpawnParticlesInGrid(X Count=30, Y Count=30, Z Count=0)
  ]
  particleSpawnScript [
   &PartitionParticles(Number Of Partitions=3, Partition Mode:ENiagara_PartitionMode=::NewEnumerator5) {module=PartitionParticles001}
   &self:PartitionByBuildingSize_02(
    Fraction_MediumB=1
    Franction_TallB=1
    LBuilding_IndexMax=1
    LBuilding_IndexMin=1
    MBuilding_IndexMax=1
    MBuilding_IndexMin=1
    SBuilding_IndexMax=1
    SBuilding_IndexMin=1
   ) {module=PartitionByBuildingSize_02}
   &FN_SetSpawnGroup(Number of Spawn Groups=6) {module=FN_SetSpawnGroup001}
   &InitializeParticle(
    Mesh Scale=<0.065 0.065 0.1>
    Mesh Scale Mode=::NewEnumerator4
    Position Offset Coordinate Space=::Local
    UsePositionOffset=true
    Write Material Random=true
   )
   /Niagara/Modules/Spawn/Location/V2/GridLocation(
    Coordinate Space:ENiagaraCoordinateSpace=::World
    Module.Jitter Inside Grid Cells=false
    Module.Override Random Seed=false
    Normalize Offsets=false
    XYZ Dimensions:Vector3f=&MakeVector(
     X=&Divide_Float(A=$Emitter.VolumeRadius, B=$Emitter.PaddingPerCell)
     Y=&Divide_Float(A=$Emitter.VolumeRadius, B=$Emitter.PaddingPerCell)
     Z=0
    )
   )
   &InitialMeshOrientation(
    Mesh Orientation Mode=::NewEnumerator2
    Rotation=&SelectVectorFromArray(
     Array Sampling Mode=::NewEnumerator0
     Direct Array Index=?
     Vector Selection Array=DataInterface() {propertyValues {FloatData [<0 0 0>, <0 0 0.25>, <0 0 0.5>]}}
     Random Seed=0
    )
    Use Rotation Vector=true
   )
  ]
  particleUpdateScript [
   &ParticleState(Kill Particles When Lifetime Has Elapsed=false, Let Infinitely Lived Particles Die When Emitter Deactivates=true)
   &SkeletalMeshLocation(
    Bone Sampling Mode=::NewEnumerator4
    MaskSpawnGroup=false
    Orientation Sampling=::NewEnumerator1
    Position Sampling=::NewEnumerator2
    Sampled Orientation Coordinate Space=::World
    Sampled Position Coordinate Space=::World
    Sampled Position Offset Coordinate Space:ENiagaraSkelMeshTransforms=::NewEnumerator0
    Sampled Scale Coordinate Space:ENiagaraCoordinateSpace=::Simulation
    Skeletal Mesh=DataInterface() {propertyValues {
      DefaultMesh=null
      ExcludeBoneName=root
      OptionalComponentClass=null
      PreviewMesh=&SK_M_SML_Ninja_04_Body
      SkinningMode=SkinOnTheFly
      SoftSourceActor=null
      SourceMode=AttachParent
      UvSetIndex=0
      WholeMeshLOD=-1
      bExcludeBone=false
      bReadDeformedGeometry=true
      bRequireCurrentFrameData=true
      ComponentTags []
      FilteredBones [pelvis,neck_02]
      FilteredSockets []
      MeshUserParameter {parameter {name=None, typeDefHandle {registeredTypeIndex=1}, varData []}}
      SamplingRegions []
     }}
   )
   /SetVariables(Particles.PelvisPosition:NiagaraPosition=$Output.SkeletalMeshLocation.SampledPositionWithoutOffset) {isSetParameters=true}
   &SkeletalMeshLocation(
    Bone / Socket Index=1
    Bone Sampling Mode=::NewEnumerator4
    MaskSpawnGroup=false
    Orientation Sampling=::NewEnumerator1
    Position Sampling=::NewEnumerator2
    Sampled Position Coordinate Space=::World
    Sampled Position Offset Coordinate Space:ENiagaraSkelMeshTransforms=::NewEnumerator0
    Sampled Scale Coordinate Space:ENiagaraCoordinateSpace=::Simulation
    Skeletal Mesh=DataInterface() {propertyValues {
      DefaultMesh=null
      ExcludeBoneName=root
      OptionalComponentClass=null
      PreviewMesh=&SK_M_SML_Ninja_04_Body
      SkinningMode=SkinOnTheFly
      SoftSourceActor=null
      SourceMode=AttachParent
      UvSetIndex=0
      WholeMeshLOD=-1
      bExcludeBone=false
      bReadDeformedGeometry=true
      bRequireCurrentFrameData=true
      ComponentTags []
      FilteredBones [pelvis,head]
      FilteredSockets []
      MeshUserParameter {parameter {name=None, typeDefHandle {registeredTypeIndex=1}, varData []}}
      SamplingRegions []
     }}
   ) {module=SkeletalMeshLocation001}
   /SetVariables(Particles.NeckPosition:NiagaraPosition=$Output.SkeletalMeshLocation001.SampledPositionWithoutOffset) {isSetParameters=true}
   &SkeletalMeshLocation(
    Bone / Socket Index=2
    Bone Sampling Mode=::NewEnumerator4
    MaskSpawnGroup=false
    Orientation Sampling=::NewEnumerator1
    Position Sampling=::NewEnumerator2
    Sampled Position Offset Coordinate Space:ENiagaraSkelMeshTransforms=::NewEnumerator0
    Sampled Scale Coordinate Space:ENiagaraCoordinateSpace=::Simulation
    Skeletal Mesh=DataInterface() {propertyValues {
      DefaultMesh=null
      ExcludeBoneName=root
      OptionalComponentClass=null
      PreviewMesh=&SK_M_SML_Ninja_04_Body
      SkinningMode=SkinOnTheFly
      SoftSourceActor=null
      SourceMode=AttachParent
      UvSetIndex=0
      WholeMeshLOD=-1
      bExcludeBone=false
      bReadDeformedGeometry=true
      bRequireCurrentFrameData=true
      ComponentTags []
      FilteredBones [pelvis,neck_02,foot_l,ik_foot_l]
      FilteredSockets []
      MeshUserParameter {parameter {name=None, typeDefHandle {registeredTypeIndex=1}, varData []}}
      SamplingRegions []
     }}
   ) {module=SkeletalMeshLocation002}
   /SetVariables(Particles.LeftLegPosition:NiagaraPosition=$Output.SkeletalMeshLocation002.SampledPositionWithoutOffset) {isSetParameters=true}
   &SkeletalMeshLocation(
    Bone / Socket Index=3
    Bone Sampling Mode=::NewEnumerator4
    MaskSpawnGroup=false
    Orientation Sampling=::NewEnumerator1
    Position Sampling=::NewEnumerator2
    Sampled Position Offset Coordinate Space:ENiagaraSkelMeshTransforms=::NewEnumerator0
    Sampled Scale Coordinate Space:ENiagaraCoordinateSpace=::Simulation
    Skeletal Mesh=DataInterface() {propertyValues {
      DefaultMesh=null
      ExcludeBoneName=root
      OptionalComponentClass=null
      PreviewMesh=&SK_M_SML_Ninja_04_Body
      SkinningMode=SkinOnTheFly
      SoftSourceActor=null
      SourceMode=AttachParent
      UvSetIndex=0
      WholeMeshLOD=-1
      bExcludeBone=false
      bReadDeformedGeometry=true
      bRequireCurrentFrameData=true
      ComponentTags []
      FilteredBones [pelvis,neck_02,calf_r,ik_foot_r]
      FilteredSockets []
      MeshUserParameter {parameter {name=None, typeDefHandle {registeredTypeIndex=1}, varData []}}
      SamplingRegions []
     }}
   ) {module=SkeletalMeshLocation003}
   /SetVariables(Particles.RightLegPosition:NiagaraPosition=$Output.SkeletalMeshLocation003.SampledPositionWithoutOffset) {isSetParameters=true}
   &SkeletalMeshLocation(
    Bone / Socket Index=3
    Bone Sampling Mode=::NewEnumerator4
    MaskSpawnGroup=false
    Orientation Sampling=::NewEnumerator1
    Position Sampling=::NewEnumerator2
    Sampled Position Offset Coordinate Space:ENiagaraSkelMeshTransforms=::NewEnumerator0
    Sampled Scale Coordinate Space:ENiagaraCoordinateSpace=::Simulation
    Skeletal Mesh=DataInterface() {propertyValues {
      DefaultMesh=null
      ExcludeBoneName=root
      OptionalComponentClass=null
      PreviewMesh=&SK_M_SML_Ninja_04_Body
      SkinningMode=SkinOnTheFly
      SoftSourceActor=null
      SourceMode=AttachParent
      UvSetIndex=0
      WholeMeshLOD=-1
      bExcludeBone=false
      bReadDeformedGeometry=true
      bRequireCurrentFrameData=true
      ComponentTags []
      FilteredBones [calf_l]
      FilteredSockets []
      MeshUserParameter {parameter {name=None, typeDefHandle {registeredTypeIndex=1}, varData []}}
      SamplingRegions []
     }}
   ) {module=SkeletalMeshLocation004}
   /SetVariables(Particles.LeftKneePosition:NiagaraPosition=$Particles.SkeletalMeshLocation004.SampledPosition) {isSetParameters=true}
   &SkeletalMeshLocation(
    Bone / Socket Index=3
    Bone Sampling Mode=::NewEnumerator4
    MaskSpawnGroup=false
    Orientation Sampling=::NewEnumerator1
    Position Sampling=::NewEnumerator2
    Sampled Position Offset Coordinate Space:ENiagaraSkelMeshTransforms=::NewEnumerator0
    Sampled Scale Coordinate Space:ENiagaraCoordinateSpace=::Simulation
    Skeletal Mesh=DataInterface() {propertyValues {
      DefaultMesh=null
      ExcludeBoneName=root
      OptionalComponentClass=null
      PreviewMesh=&SK_M_SML_Ninja_04_Body
      SkinningMode=SkinOnTheFly
      SoftSourceActor=null
      SourceMode=AttachParent
      UvSetIndex=0
      WholeMeshLOD=-1
      bExcludeBone=false
      bReadDeformedGeometry=true
      bRequireCurrentFrameData=true
      ComponentTags []
      FilteredBones [calf_r]
      FilteredSockets []
      MeshUserParameter {parameter {name=None, typeDefHandle {registeredTypeIndex=1}, varData []}}
      SamplingRegions []
     }}
   ) {module=SkeletalMeshLocation005}
   /SetVariables(Particles.RightKneePosition:NiagaraPosition=$Output.SkeletalMeshLocation005.SampledPositionWithoutOffset) {isSetParameters=true}
   &SolveForcesAndVelocity(Clamp Velocity=false, Limit Acceleration=false)
   &ScaleColor(
    Scale Alpha=&Lerp_Float(
     A=1
     B=0
     Alpha=&RemapRange(
      Input Value To Remap=&DistanceBetweenPositions(A=$Particles.Position, B=&SimulationPosition())
      Input Min=130
      Input Max=180
      Output Min=0
      Output Max=1
     )
    )
    Scale RGB=&Lerp_Vector(
     A=&MakeVectorFromLinearColorRGB(Color=<rgba 0.1048 5 1.9461 1>)
     B=&MakeVectorFromLinearColorRGB(Color=<rgba 1 0.8542 0.8542 1>)
     Alpha=&RemapRange(
      Input Value To Remap=&DistanceBetweenPositions(A=$Particles.Position, B=Linked() {linkedVariable {name=Particles.PelvisPosition, type=Vector3f}})
      Input Min=0
      Input Max=150
      Output Min=0
      Output Max=1
     )
    )
    ScaleA=true
    ScaleRGB=true
   )
   /SetVariables(Particles.PositionCenter:NiagaraPosition=&AddVectorToPosition(Position=$Engine.Owner.Position, Vector=$Emitter.OriginOffset)) {isSetParameters=true}
   &self:TreadmillParticles_01(
    GridSize:NiagaraFloat=&Multiply_Float(
     A=$Emitter.VolumeRadius
     B=&Lerp_Float(
      A=3
      B=2.5
      Alpha=&MakeCustomFloatFromBool(
       Bool=&SetBoolByFloatComparison(A=&VectorLength(Vector=$User.OwnerLocalVelocity), B=1)
       True Float=1
       False Float=0
      )
     )
    )
    InputPosition:NiagaraPosition=$Particles.PositionCenter
   ) {module=TreadmillParticles_01}
   /SetVariables(Particles.Position=&MakeVector(
     X=&MakeFloatFromVector(Vector=Linked() {linkedVariable {name=Particles.Position, type=NiagaraPosition}}, Channel=::NewEnumerator0)
     Y=&MakeFloatFromVector(Vector=Linked() {linkedVariable {name=Particles.Position, type=NiagaraPosition}}, Channel=::NewEnumerator1)
     Z=&MakeFloatFromVector(Vector=&SimulationPosition(), Channel=::NewEnumerator2)
    )) {isSetParameters=true}
   /SetVariables(Particles.Position=&AddVectorToPosition(Position=$Particles.Position, Vector=<0 0 100>)) {isSetParameters=true}
   /SetVariables(Particles.Position=&AddVectorToPosition(
     Position=$Particles.Position
     Vector=&FN_FrontEndVector3Modifier(
      Gameplay=<0 0 0>
      Front End=&RotateVectorByQuaternion(Vector To Rotate=<0 -1.2 0>, Quaternion=$Output.SocketLocation.SampledRotation)
     )
    )) {isSetParameters=true}
   /SetVariables(Particles.LerpedAudioReactive:NiagaraFloat=&self:Lerp_Audio_Reactive_EQ(
     Default:Vector3f=&VectorFromFloat(Value=$NPC.FortniteNPC.EventAudio_Amplitude)
     ChangeRate:NiagaraFloat=30
    )) {isSetParameters=true}
   &Timeline(
    Additional Output:ENiagara_TimelineOutput=::NewEnumerator1
    Additional Output Mode:ENiagara_TimelineOutputMode=::NewEnumerator1
    Float Curve=DataInterface() {propertyValues {Curve {keys [
        {arriveTangent=0.9218, interpMode=RCIM_Cubic, leaveTangent=0.9218, tangentMode=RCTM_User, time=0.35, value=0}
        {arriveTangent=2.875, interpMode=RCIM_Cubic, leaveTangent=2.875, time=0.63, value=0.9}
        {arriveTangent=-0.0156, interpMode=RCIM_Cubic, leaveTangent=-0.0156, tangentMode=RCTM_User, time=0.75, value=1.15}
        {arriveTangent=-0.0297, interpMode=RCIM_Cubic, leaveTangent=-0.0297, tangentMode=RCTM_User, time=0.9, value=0.95}
        {arriveTangent=0.0389, interpMode=RCIM_Cubic, leaveTangent=0.0389, tangentMode=RCTM_User, time=0.96, value=1.025}
        <at 1 1 RCIM_Cubic>
       ]}}}
    Play=&Boolean_NotOperation(Bool=$Particles.RegionMask.IsInsideMask)
    Play Rate=&MakeCustomFloatFromBool(
     Bool=&SetBoolByFloatComparison(A=&VectorLength(Vector=$User.OwnerLocalVelocity), B=160)
     True Float=50
     False Float=2
    )
    Rewind Rate:NiagaraFloat=&MakeCustomFloatFromBool(
     Bool=&SetBoolByFloatComparison(A=&VectorLength(Vector=$User.OwnerLocalVelocity), B=160)
     True Float=10
     False Float=3
    )
   )
   &self:GridCollision(TraceDir:Vector3f=<0 0 -1>, TraceDist=250, ZBoost=250, bIsFrontEnd:NiagaraBool=$User.bIsFrontEndPreview) {module=GridCollision}
   /SetVariables(Particles.TargetPos:Vector3f=&self:Return_Closest_Joint(
     PelvisPosition:Vector3f=$Particles.PelvisPosition
     NeckPosition:Vector3f=$Particles.NeckPosition
     LeftLegPosition:Vector3f=$Particles.LeftLegPosition
     RightLegPosition:Vector3f=$Particles.RightLegPosition
     LeftCalfPosition:Vector3f=$Particles.LeftKneePosition
     RightCalfPosition:Vector3f=$Particles.RightKneePosition
     NeckOffset:Vector3f=<0 0 -25>
     PelvisOffset:Vector3f=<0 0 -10>
     KneeOffset:Vector3f=<0 0 -10>
     FootOffset:Vector3f=<0 0 -5>
     ParticlePosition:Vector3f=Linked() {linkedVariable {name=Particles.Position, type=NiagaraPosition}}
    )) {isSetParameters=true}
   &RegionMask(
    Invert Region=true
    Offset:Vector3f=<0 10 0>
    Offset Space:ENiagaraCoordinateSpace=::Local
    Outer Float=0
    Region Dimensions:Vector3f=<225 500 1500>
    Region Falloff=300
    Region Type:ENiagaraDebugDrawShape=::NewEnumerator1
    Rotation:Vector3f=$Engine.Owner.SystemYAxis
    Rotation Angle Type:ENiagara_AngleInput=::NewEnumerator0
    UseOffset=true
    UseRotation=true
   )
   &ScaleMeshSize(Scale Factor=&VectorFromFloat(Value=&Multiply_Float(
      A=$Particles.Timeline.Float
      B=&FloatFromCurve(FloatCurve=DataInterface() {propertyValues {Curve {
          defaultValue=1
          keys [{interpMode=RCIM_Cubic, tangentMode=RCTM_User, time=0, value=1}, {interpMode=RCIM_Cubic, tangentMode=RCTM_User, time=1, value=1}]
         }}})
     ))) {module=ScaleMeshSize001}
   &ScaleMeshSize(Scale Factor=&MakeVector(
     X=1
     Y=1
     Z=&Multiply_Float(
      A=&self:Lerp_Player_Position_EQ(
       TargetPos:Vector3f=$Particles.TargetPos
       XY Distance:NiagaraFloat=100
       ParticlePosition:Vector3f=Linked() {linkedVariable {name=Particles.CollisionPos, type=NiagaraPosition}}
       BaseScale:NiagaraFloat=10
       InputMinTest:NiagaraFloat=0
       TargetZOffset:NiagaraFloat=0
      )
      B=0.1
     )
    )) {module=ScaleMeshSize002}
   &ScaleMeshSize(Scale Factor=&Lerp_Vector(
     A=<1 1 1>
     B=<1 1 2.5>
     Alpha=&RemapRange(
      Input Value To Remap=&Multiply_Float(
       A=$Particles.LerpedAudioReactive
       B=&Multiply_Float(
        A=&RandomRangeFloat(Minimum=0.8, Maximum=1)
        B=&self:Height_Gate(
         ChangeRate:NiagaraFloat=10
         Max:NiagaraFloat=0.4
         Particle Z Scale:NiagaraFloat=&MakeFloatFromVector(Vector=$Particles.Scale, Channel=::NewEnumerator2)
        )
       )
      )
      Input Min=0.7
      Input Max=1
      Output Min=0
      Output Max=1
     )
    )) {module=ScaleMeshSize003}
   &ScaleMeshSize(Scale Factor=&Lerp_Vector(
     A=<1 1 1>
     B=&Lerp_Vector(
      A=<1 1 1>
      B=<1 1 2>
      Alpha=&Lerp_Float(
       A=0
       B=&Add_Float(
        A=&Sine(
         Normalized Angle=&Add_Float(
          A=$Emitter.Age
          B=&Multiply_Float(
           A=&DistanceBetweenPositions(A=$Particles.Position, B=Linked() {linkedVariable {name=Particles.PelvisPosition, type=Vector3f}})
           B=-0.01
          )
         )
         Period=1
         Scale=1
         Bias=0
        )
        B=1
       )
       Alpha=&RemapRange(
        Input Value To Remap=&DistanceBetweenPositions(A=Linked() {linkedVariable {name=Particles.PelvisPosition, type=Vector3f}}, B=$Particles.Position)
        Input Min=0
        Input Max=150
        Output Min=0
        Output Max=1
       )
      )
     )
     Alpha=&self:Height_Gate(
      ChangeRate:NiagaraFloat=12
      Max:NiagaraFloat=0.11
      Particle Z Scale:NiagaraFloat=&MakeFloatFromVector(Vector=$Particles.Scale, Channel=::NewEnumerator2)
     )
    ))
  ]
  rendererClasses [&NiagaraMeshRendererProperties]
  renderers [{rendererClass=&NiagaraMeshRendererProperties, rendererIndex=0}]
 }
}
summary {
 systemName=NS_FareSporeCookie_Debug
 emitters [
  {bEnabled=false, emitterName=Visualizer, rendererClasses [&NiagaraMeshRendererProperties]}
  {emitterName=Square_far001, rendererClasses [&NiagaraMeshRendererProperties]}
  {bEnabled=false, emitterName=BlowingParticles, rendererClasses [&NiagaraLightRendererProperties]}
  {emitterName=FlareBurst, rendererClasses [&NiagaraSpriteRendererProperties]}
  {emitterName=Glow, rendererClasses [&NiagaraSpriteRendererProperties]}
 ]
 userVariables [
  User.AltTimeline {type=NiagaraFloat, defaultValue=0, description=""}
  User.bIsFrontEnd {type=NiagaraBool}
  User.bIsFrontEndPreview {type=NiagaraBool}
  User.FrontEnd_color {type=LinearColor, description="", defaultValue {struct=LinearColor, value=<rgba 1 0.6743 0.6445 1>}}
  User.HandGlow {type=NiagaraFloat, defaultValue=0, description=""}
  User.OwnerLocalVelocity {type=Vector3f, description="", defaultValue {struct=Vector3f, value=<1 1 1>}}
  User.SkeletalMesh {
   type=NiagaraDataInterfaceSkeletalMesh
   defaultValue {
    struct=NiagaraExt_VariableValue_DataInterface
    value {dataInterface=&self:NiagaraDataInterfaceSkeletalMesh_0}
   }
  }
 ]
}
