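# Test Niagara system for the FareSporeCookie emote, in an Old folder, with only the Audio_Visualizer mesh-grid emitter enabled.
# Grid particles run through treadmill and ground-trace collision modules, then a radius mask drives a timeline that scales meshes in and out.
# Skeletal bone positions and smoothed audio amplitude are computed, but their distance, color and audio-stretch consumers are disabled.
# Disabled emitters hold another grid, audio-reactive boxes, wind-blown GPU sprites and a debug sprite.
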NiagaraSystem /BRCosmetics/Animation/Game/MainPlayer/Emotes/FareSporeCookie/FX/Old/NS_FareSporeCookie_Test
dependencies {
 usedDataInterfaces [
  &NiagaraDataInterfaceCurve
  &NiagaraDataInterfaceMeshRendererInfo
  &NiagaraDataInterfaceArrayFloat3
  &NiagaraDataInterfaceSkeletalMesh
  &NiagaraDataInterfaceCollisionQuery
  &NiagaraDataInterfaceDebugDraw
  &NiagaraDataInterfaceArrayFloat
  &NiagaraDataInterfaceColorCurve
  &NiagaraDataInterfaceLandscape
 ]
 usedDynamicInputs [&MakeIntFromFloat]
 usedModules [&SystemState]
}
emitters {
 Audio_Visualizer {
  emitterUpdateScript [
   &EmitterState(
    Inactive Response=::NewEnumerator1
    Life Cycle Mode=::NewEnumerator1
    Loop Behavior=::NewEnumerator1
    Loop Duration Mode=::NewEnumerator1
    UseLoopDelay=false
   )
   /SetVariables(
    Emitter.OriginOffset:Vector3f=&TransformVector(Vector=<0 0 0>, Destination Space=::Local)
    Emitter.PaddingPerCell=10
    Emitter.VolumeRadius=200
   ) {isSetParameters=true, module=SetVariables_#1}
   /Niagara/Modules/Spawn/Location/SpawnParticlesInGrid(X Count=30, Y Count=30, Z Count=0)
  ]
  particleSpawnScript [
   &PartitionParticles(Number Of Partitions=3, Partition Mode:ENiagara_PartitionMode=::NewEnumerator5) {module=PartitionParticles001}
   &self:PartitionByBuildingSize_03(
    Fraction_MediumB=1
    Franction_TallB=1
    LBuilding_IndexMax=1
    LBuilding_IndexMin=1
    MBuilding_IndexMax=1
    MBuilding_IndexMin=1
    SBuilding_IndexMax=1
    SBuilding_IndexMin=1
   ) {module=PartitionByBuildingSize_03}
   &FN_SetSpawnGroup(Number of Spawn Groups=6) {module=FN_SetSpawnGroup001}
   &InitializeParticle(
    Mesh Scale Mode=::NewEnumerator3
    Mesh Uniform Scale=0.1
    Position Offset Coordinate Space=::Local
    UsePositionOffset=true
    Write Material Random=true
   )
   /Niagara/Modules/Spawn/Location/V2/GridLocation(
    Coordinate Space:ENiagaraCoordinateSpace=::World
    Module.Jitter Inside Grid Cells=false
    Module.Override Random Seed=false
    Normalize Offsets=false
    XYZ Dimensions:Vector3f=&MakeVector(
     X=&Divide_Float(A=$Emitter.VolumeRadius, B=$Emitter.PaddingPerCell)
     Y=&Divide_Float(A=$Emitter.VolumeRadius, B=$Emitter.PaddingPerCell)
     Z=0
    )
   )
   &InitialMeshOrientation(
    Mesh Orientation Mode=::NewEnumerator2
    Rotation=&SelectVectorFromArray(
     Array Sampling Mode=::NewEnumerator0
     Direct Array Index=?
     Vector Selection Array=DataInterface() {propertyValues {FloatData [<0 0 0>, <0 0 0.25>, <0 0 0.5>]}}
     Random Seed=0
    )
    Use Rotation Vector=true
   )
  ]
  particleUpdateScript [
   &ParticleState(Kill Particles When Lifetime Has Elapsed=false, Let Infinitely Lived Particles Die When Emitter Deactivates=true)
   &SkeletalMeshLocation(
    Bone Sampling Mode=::NewEnumerator4
    MaskSpawnGroup=false
    Orientation Sampling=::NewEnumerator1
    Position Sampling=::NewEnumerator2
    Sampled Position Offset Coordinate Space:ENiagaraSkelMeshTransforms=::NewEnumerator1
    Skeletal Mesh=DataInterface() {propertyValues {
      DefaultMesh=null
      ExcludeBoneName=root
      OptionalComponentClass=null
      PreviewMesh=/Game/Characters/Player/Male/Small/Bodies/M_Sml_Ninja_04/Meshes/SK_M_SML_Ninja_04_Body
      SkinningMode=SkinOnTheFly
      SoftSourceActor=null
      SourceMode=AttachParent
      UvSetIndex=0
      WholeMeshLOD=-1
      bExcludeBone=false
      bReadDeformedGeometry=true
      bRequireCurrentFrameData=true
      ComponentTags []
      FilteredBones [pelvis,neck_02]
      FilteredSockets []
      MeshUserParameter {parameter {name=None, typeDefHandle {registeredTypeIndex=1}, varData []}}
      SamplingRegions []
     }}
   )
   /SetVariables(Particles.PelvisPosition:NiagaraPosition=$Output.SkeletalMeshLocation.SampledPositionWithoutOffset) {isSetParameters=true}
   &SkeletalMeshLocation(
    Bone / Socket Index=1
    Bone Sampling Mode=::NewEnumerator4
    MaskSpawnGroup=false
    Orientation Sampling=::NewEnumerator1
    Position Sampling=::NewEnumerator2
    Sampled Position Offset Coordinate Space:ENiagaraSkelMeshTransforms=::NewEnumerator1
    Skeletal Mesh=DataInterface() {propertyValues {
      DefaultMesh=null
      ExcludeBoneName=root
      OptionalComponentClass=null
      PreviewMesh=&SK_M_SML_Ninja_04_Body
      SkinningMode=SkinOnTheFly
      SoftSourceActor=null
      SourceMode=AttachParent
      UvSetIndex=0
      WholeMeshLOD=-1
      bExcludeBone=false
      bReadDeformedGeometry=true
      bRequireCurrentFrameData=true
      ComponentTags []
      FilteredBones [pelvis,neck_02]
      FilteredSockets []
      MeshUserParameter {parameter {name=None, typeDefHandle {registeredTypeIndex=1}, varData []}}
      SamplingRegions []
     }}
   ) {module=SkeletalMeshLocation001}
   /SetVariables(Particles.NeckPosition:NiagaraPosition=$Output.SkeletalMeshLocation001.SampledPositionWithoutOffset) {isSetParameters=true}
   &SkeletalMeshLocation(
    Bone / Socket Index=2
    Bone Sampling Mode=::NewEnumerator4
    MaskSpawnGroup=false
    Orientation Sampling=::NewEnumerator1
    Position Sampling=::NewEnumerator2
    Sampled Position Offset Coordinate Space:ENiagaraSkelMeshTransforms=::NewEnumerator1
    Skeletal Mesh=DataInterface() {propertyValues {
      DefaultMesh=null
      ExcludeBoneName=root
      OptionalComponentClass=null
      PreviewMesh=&SK_M_SML_Ninja_04_Body
      SkinningMode=SkinOnTheFly
      SoftSourceActor=null
      SourceMode=AttachParent
      UvSetIndex=0
      WholeMeshLOD=-1
      bExcludeBone=false
      bReadDeformedGeometry=true
      bRequireCurrentFrameData=true
      ComponentTags []
      FilteredBones [pelvis,neck_02,foot_l,foot_r]
      FilteredSockets []
      MeshUserParameter {parameter {name=None, typeDefHandle {registeredTypeIndex=1}, varData []}}
      SamplingRegions []
     }}
   ) {module=SkeletalMeshLocation002}
   /SetVariables(Particles.LeftLegPosition:NiagaraPosition=$Output.SkeletalMeshLocation002.SampledPositionWithoutOffset) {isSetParameters=true}
   &SkeletalMeshLocation(
    Bone / Socket Index=3
    Bone Sampling Mode=::NewEnumerator4
    MaskSpawnGroup=false
    Orientation Sampling=::NewEnumerator1
    Position Sampling=::NewEnumerator2
    Sampled Position Offset Coordinate Space:ENiagaraSkelMeshTransforms=::NewEnumerator1
    Skeletal Mesh=DataInterface() {propertyValues {
      DefaultMesh=null
      ExcludeBoneName=root
      OptionalComponentClass=null
      PreviewMesh=&SK_M_SML_Ninja_04_Body
      SkinningMode=SkinOnTheFly
      SoftSourceActor=null
      SourceMode=AttachParent
      UvSetIndex=0
      WholeMeshLOD=-1
      bExcludeBone=false
      bReadDeformedGeometry=true
      bRequireCurrentFrameData=true
      ComponentTags []
      FilteredBones [pelvis,neck_02,calf_r,calf_l]
      FilteredSockets []
      MeshUserParameter {parameter {name=None, typeDefHandle {registeredTypeIndex=1}, varData []}}
      SamplingRegions []
     }}
   ) {module=SkeletalMeshLocation003}
   /SetVariables(Particles.RightLegPosition:NiagaraPosition=$Output.SkeletalMeshLocation003.SampledPositionWithoutOffset) {isSetParameters=true}
   /SetVariables(Particles.PositionCenter:NiagaraPosition=&AddVectorToPosition(Position=$Engine.Owner.Position, Vector=<0 0 0>)) {isSetParameters=true, module=SetVariables_#2}
   &SolveForcesAndVelocity(Clamp Velocity=false, Limit Acceleration=false)
   &ScaleColor(
    Scale Alpha=&Lerp_Float(
     A=1
     B=0
     Alpha=&RemapRange(
      Input Value To Remap=&DistanceBetweenPositions(A=$Particles.Position, B=/Niagara/DynamicInputs/Helpers/SimulationPosition())
      Input Min=180
      Input Max=200
      Output Min=0
      Output Max=1
     )
    )
    Scale RGB=&Lerp_Vector(
     A=&MakeVectorFromLinearColorRGB(Color=<rgba 0.1048 5 1.9461 1>)
     B=&MakeVectorFromLinearColorRGB(Color=<rgba 1 0.8542 0.8542 1>)
     Alpha=&RemapRange(
      Input Value To Remap=&DistanceBetweenPositions(A=$Particles.Position, B=Linked() {linkedVariable {name=Particles.PelvisPosition, type=Vector3f}})
      Input Min=0
      Input Max=200
      Output Min=0
      Output Max=1
     )
    )
    ScaleA=false
    ScaleRGB=true
   ) {enabled=false}
   &self:TreadmillParticles_01(
    GridSize:NiagaraFloat=&Multiply_Float(A=$Emitter.VolumeRadius, B=3)
    InputPosition:NiagaraPosition=$Particles.PositionCenter
   ) {module=TreadmillParticles_01}
   /SetVariables(Particles.Position=&MakeVector(
     X=&MakeFloatFromVector(Vector=<0 0 0>, Channel=::NewEnumerator0)
     Y=&MakeFloatFromVector(Vector=<0 0 0>, Channel=::NewEnumerator1)
     Z=&MakeFloatFromVector(Vector=<0 0 0>, Channel=::NewEnumerator2)
    )) {isSetParameters=true, module=SetVariables_#3}
   /SetVariables(Particles.Position=&AddVectorToPosition(Position=$Particles.Position, Vector=<0 0 100>)) {isSetParameters=true, module=SetVariables_#4}
   /SetVariables(Particles.Position=&AddVectorToPosition(Position=$Particles.Position, Vector=<0 0 0>)) {isSetParameters=true, module=SetVariables_#5}
   &self:GridCollision_01(TraceDir:Vector3f=<0 0 -1>, TraceDist=250, ZBoost=250, bIsFrontEnd:NiagaraBool=$User.bIsFrontEndPreview) {module=GridCollision_01}
   &RegionMask(
    Invert Region=true
    Offset:Vector3f=$Emitter.OriginOffset
    Outer Float=0
    Region Falloff=0
    Region Radius:NiagaraFloat=&Add_Float(
     A=&Multiply_Float(A=$Emitter.VolumeRadius, B=1)
     B=&Multiply_Float(A=$Emitter.VolumeRadius, B=0)
    )
    UseOffset=true
    UseRotation=false
   )
   &Timeline(
    Additional Output:ENiagara_TimelineOutput=::NewEnumerator1
    Additional Output Mode:ENiagara_TimelineOutputMode=::NewEnumerator1
    Float Curve=DataInterface() {propertyValues {Curve {keys [
        {arriveTangent=0.9218, interpMode=RCIM_Cubic, leaveTangent=0.9218, tangentMode=RCTM_User, time=0.35, value=0}
        {arriveTangent=2.875, interpMode=RCIM_Cubic, leaveTangent=2.875, time=0.63, value=0.9}
        {arriveTangent=-0.0156, interpMode=RCIM_Cubic, leaveTangent=-0.0156, tangentMode=RCTM_User, time=0.75, value=1.15}
        {arriveTangent=-0.0297, interpMode=RCIM_Cubic, leaveTangent=-0.0297, tangentMode=RCTM_User, time=0.9, value=0.95}
        {arriveTangent=0.0389, interpMode=RCIM_Cubic, leaveTangent=0.0389, tangentMode=RCTM_User, time=0.96, value=1.025}
        <at 1 1 RCIM_Cubic>
       ]}}}
    Play=&Boolean_NotOperation(Bool=$Particles.RegionMask.IsInsideMask)
    Play Rate=&MakeCustomFloatFromBool(
     Bool=&SetBoolByFloatComparison(A=&VectorLength(Vector=<0 0 0>), B=160)
     True Float=30
     False Float=2
    )
    Rewind Rate:NiagaraFloat=&MakeCustomFloatFromBool(
     Bool=&SetBoolByFloatComparison(A=&VectorLength(Vector=<0 0 0>), B=150)
     True Float=5
     False Float=30
    )
   )
   /SetVariables(Particles.TargetPosition:NiagaraFloat=/Niagara/DynamicInputs/Math/CompareFloats(
     Comparison Mode:CompareValues=::NewEnumerator1
     Comparison Count:ENiagara_MultipleLerpCount=::NewEnumerator1
     A:NiagaraFloat=&DistanceBetweenPositions(
      A=&Multiply_Vector(A=Linked() {linkedVariable {name=Particles.Position, type=NiagaraPosition}}, B=<1 1 0>)
      B=&Multiply_Vector(A=$Particles.PelvisPosition, B=<1 1 0>)
     )
     B:NiagaraFloat=&DistanceBetweenPositions(
      A=&Multiply_Vector(A=Linked() {linkedVariable {name=Particles.Position, type=NiagaraPosition}}, B=<1 1 0>)
      B=&Multiply_Vector(A=$Particles.NeckPosition, B=<1 1 0>)
     )
     C:NiagaraFloat=&DistanceBetweenPositions(
      A=&Multiply_Vector(A=Linked() {linkedVariable {name=Particles.Position, type=NiagaraPosition}}, B=<1 1 0>)
      B=&Multiply_Vector(A=$Particles.LeftLegPosition, B=<1 1 0>)
     )
     D:NiagaraFloat=&DistanceBetweenPositions(
      A=&Multiply_Vector(A=Linked() {linkedVariable {name=Particles.Position, type=NiagaraPosition}}, B=<1 1 0>)
      B=&Multiply_Vector(A=$Particles.RightLegPosition, B=<1 1 0>)
     )
     E:NiagaraFloat=?
    )) {enabled=false, isSetParameters=true}
   /SetVariables(Particles.TargetPosition:NiagaraFloat=&self:Return_Smallest_Distance(
     PelvisPosition:Vector3f=$Particles.PelvisPosition
     NeckPosition:Vector3f=$Particles.NeckPosition
     LeftLegPosition:Vector3f=$Particles.LeftLegPosition
     RightLegPosition:Vector3f=$Particles.RightLegPosition
     ParticlePosition:Vector3f=Linked() {linkedVariable {name=Particles.Position, type=NiagaraPosition}}
    )) {enabled=false, isSetParameters=true}
   /SetVariables(Particles.TargetLength:NiagaraFloat=$Particles.TargetPosition) {enabled=false, isSetParameters=true}
   /SetVariables(Particles.LerpedAudioReactive:NiagaraFloat=&self:Lerp_Audio_Reactive_EQ(Default:Vector3f=&VectorFromFloat(Value=$NPC.FortniteNPC.EventAudio_Amplitude))) {isSetParameters=true}
   &ScaleMeshSize(Scale Factor=&VectorFromFloat(Value=&Multiply_Float(
      A=$Particles.Timeline.Float
      B=&FloatFromCurve(FloatCurve=DataInterface() {propertyValues {Curve {
          defaultValue=1
          keys [{interpMode=RCIM_Cubic, tangentMode=RCTM_User, time=0, value=1}, {interpMode=RCIM_Cubic, tangentMode=RCTM_User, time=1, value=1}]
         }}})
     ))) {module=ScaleMeshSize001}
   &ScaleMeshSize(Scale Factor=&AddVectorToPosition(
     Position=<1 1 1>
     Vector=&Multiply_Vector(A=<0 0 1>, B=&VectorFromFloat(Value=1))
    )) {module=ScaleMeshSize002}
   &ScaleMeshSize(Scale Factor=&Lerp_Vector(
     A=<1 1 1>
     B=<1 1 2.5>
     Alpha=&RemapRange(
      Input Value To Remap=&Multiply_Float(A=$Particles.LerpedAudioReactive, B=&RandomRangeFloat(Minimum=0.8, Maximum=1))
      Input Min=0.7
      Input Max=1
      Output Min=0
      Output Max=1
     )
    )) {enabled=false, module=ScaleMeshSize003}
  ]
  rendererClasses [&NiagaraMeshRendererProperties]
  renderers [{rendererClass=&NiagaraMeshRendererProperties, rendererIndex=0}]
 }
 BlowingParticles {
  bEnabled=false
  simTarget=GPUComputeSim
  emitterUpdateScript [
   &EmitterState(Life Cycle Mode=::NewEnumerator1, UseLoopDelay=false)
   &SpawnRate(SpawnRate=500, Use Spawn Probability=false)
  ]
  particleSpawnScript [
   &InitializeParticle(
    Lifetime Max=1.5
    Lifetime Mode=::NewEnumerator1
    Mass Max=1.5
    Mass Mode=::NewEnumerator3
    Sprite Rotation Mode=::NewEnumerator2
    Sprite Size Mode=::NewEnumerator1
    Uniform Sprite Size Max=8
    Uniform Sprite Size Min=4
    UsePositionOffset=false
    Write Material Random=true
   )
   &ShapeLocation(MaskSpawnGroup=false, Override Seed=false, Sphere Radius=300, UseOwnerScale=true)
   &InitialMeshOrientation(Use Rotation Vector=false)
   &SkeletalMeshLocation(
    Bone Sampling Mode=::NewEnumerator0
    MaskSpawnGroup=false
    OverrideRandomSeed=false
    Position Sampling=::NewEnumerator1
    Skeletal Mesh=DataInterface() {propertyValues {
      DefaultMesh=null
      ExcludeBoneName=root
      OptionalComponentClass=null
      PreviewMesh=null
      SkinningMode=SkinOnTheFly
      SoftSourceActor=null
      SourceMode=Default
      UvSetIndex=0
      WholeMeshLOD=-1
      bExcludeBone=false
      bReadDeformedGeometry=true
      bRequireCurrentFrameData=true
      ComponentTags []
      FilteredBones [hand_l,hand_r,None]
      FilteredSockets []
      MeshUserParameter {parameter {name=None, typeDefHandle {registeredTypeIndex=1}, varData []}}
      SamplingRegions []
     }}
   )
  ]
  particleUpdateScript [
   &ParticleState()
   &GravityForce()
   &WindForce(
    Friction=0.8
    Offset By Time:Vector3f=<50 10 -125>
    Scale Min / Max:Vector2f=<0.275 1>
    Speed Min / Max:Vector2f=<0 500>
    Turbulence Scale Mode:ENiagaraWindTurbulenceContributionMode=::NewEnumerator1
    Wind Speed=<0 -200 0>
    Wind Speed Scale=&RandomRangeFloat(Minimum=0.333, Maximum=1)
   )
   &AerodynamicDrag(
    Aerodynamic Drag:NiagaraFloat=&RandomRangeFloat(Minimum=0.3333, Maximum=1)
    Aerodynamic Rotational Drag=0.15
    Lift Contribution=0.75
    Pivot Offset Maximum:Vector3f=<5 0.5 1>
    Pivot Offset Minimum:Vector3f=<-5 -0.5 -1>
    Rotation Strength:NiagaraFloat=&RandomRangeFloat(Minimum=1.8, Maximum=3.2)
    Use Linear Drag=true
    Use Rotational Drag=true
   )
   &Collision(
    Age Colliding Particles=false
    Enable Rest State=false
    Friction=2
    Friction During a Bounce=0.1
    Kill Offscreen Particles=true
    KillOnCollision=false
    Module.EnableMaxCollisionCount=false
    Module.Randomize Collision Normal Vector=false
    Particle Radius Scale=0.1
    Restitution=0.05
   )
   &SolveForcesAndVelocity(Clamp Velocity=false, Limit Acceleration=false)
   &ScaleSpriteSize(Uniform Curve Sprite Scale=DataInterface() {propertyValues {Curve {keys [
        {arriveTangent=14.5835, interpMode=RCIM_Cubic, leaveTangent=14.5835, tangentMode=RCTM_User, time=0, value=0}
        {arriveTangent=1.1111, interpMode=RCIM_Cubic, leaveTangent=1.1111, time=0.1, value=1}
        {arriveTangent=-1.1111, interpMode=RCIM_Cubic, leaveTangent=-1.1111, time=0.9, value=1}
        {arriveTangent=-8.7113, interpMode=RCIM_Cubic, leaveTangent=-8.7113, tangentMode=RCTM_User, time=1, value=0}
       ]}}})
   &AlignSpriteToMeshOrientation(SetAlignment=true, SetFacing=true)
   /Niagara/Modules/Collision/AlignParticlesWithCollisionPlane(Allow Alignment on Y=false, Allow Alignment on Z=false, Rotational Dampening Power=1)
   &SampleSkeletalMesh(
    AdditionalYawPitchRoll=false
    Mesh Sampling Type:ENiagaraSkelSamplingModeFull=::NewEnumerator8
    OffsetAlongNormal=false
    OverrideRandomSeed=false
    Skeletal Mesh=DataInterface() {propertyValues {
      DefaultMesh=null
      ExcludeBoneName=root
      OptionalComponentClass=null
      PreviewMesh=null
      SkinningMode=SkinOnTheFly
      SoftSourceActor=null
      SourceMode=AttachParent
      UvSetIndex=0
      WholeMeshLOD=-1
      bExcludeBone=false
      bReadDeformedGeometry=true
      bRequireCurrentFrameData=true
      ComponentTags []
      FilteredBones []
      FilteredSockets []
      MeshUserParameter {parameter {name=None, typeDefHandle {registeredTypeIndex=1}, varData []}}
      SamplingRegions []
     }}
   )
  ]
  rendererClasses [&NiagaraSpriteRendererProperties]
  renderers [{rendererClass=&NiagaraSpriteRendererProperties}]
 }
 Box {
  bEnabled=false
  emitterUpdateScript [
   &EmitterState(
    Inactive Response=::NewEnumerator1
    Life Cycle Mode=::NewEnumerator1
    Loop Behavior=::NewEnumerator1
    Loop Delay=0.5
    Loop Duration=4
    UseLoopDelay=true
   )
   &SpawnBurst_Instantaneous(Spawn Count=40, Use Spawn Probability=false, UseLoopCountLimit=false)
  ]
  particleSpawnScript [
   &PartitionParticles(Number Of Partitions=40)
   &InitializeParticle(
    Color=<rgba 0 1 0.8348 1>
    Lifetime Max=1.5
    Lifetime Min=0.5
    Lifetime Mode=::NewEnumerator1
    Mesh Scale Max=<0.8 0.8 2>
    Mesh Scale Min=<0.8 0.8 1>
    Mesh Scale Mode=::NewEnumerator2
    UsePositionOffset=false
    Write Material Random=true
   )
   &InitialMeshOrientation(
    Mesh Orientation Mode=::NewEnumerator2
    Rotation=&Multiply_Vector(
     A=&SelectVectorFromArray(
      Direct Array Index=$Particles.PartitionParticles.Partition
      Vector Selection Array=DataInterface() {propertyValues {FloatData [
         <0 0 0>
         <0 0 0.01>
         <0 0 0.02>
         <0 0 0.03>
         <0 0 0.04>
         <0 0 0.05>
         <0 0 0.06>
         <0 0 0.07>
         <0 0 0.08>
         <0 0 0.09>
         <0 0 0>
         <0 0 -0.01>
         <0 0 -0.02>
         <0 0 -0.03>
         <0 0 -0.04>
         <0 0 -0.05>
         <0 0 -0.06>
         <0 0 -0.07>
         <0 0 -0.08>
         <0 0 -0.09>
         <0 0 0.4>
         <0 0 0.41>
         <0 0 0.42>
         <0 0 0.43>
         <0 0 0.44>
         <0 0 0.45>
         <0 0 0.46>
         <0 0 0.47>
         <0 0 0.48>
         <0 0 0.49>
         <0 0 0.5>
         <0 0 0.39>
         <0 0 0.38>
         <0 0 0.37>
         <0 0 0.36>
         <0 0 0.35>
         <0 0 0.34>
         <0 0 0.33>
         <0 0 0.32>
        ]}}
     )
     B=<1 1 1.2>
    )
    Use Rotation Vector=true
   )
   /SetVariables(Particles.Randomfloat:NiagaraFloat=&RandomRangeFloat(Minimum=0, Maximum=1)) {isSetParameters=true}
  ]
  particleUpdateScript [
   &ParticleState(Kill Particles When Lifetime Has Elapsed=false, Loop Particles Lifetime=true)
   &ScaleMeshSize(Scale Factor=&Multiply_Vector(
     A=&MakeVector(X=1, Y=1, Z=&RandomRangeFloat(Minimum=5, Maximum=10))
     B=&MakeVector(
      X=1
      Y=1
      Z=&Multiply_Float(
       A=&FloatFromCurve(
        FloatCurve=DataInterface() {propertyValues {Curve {keys [
            {interpMode=RCIM_Cubic, tangentMode=RCTM_User, time=0, value=0.4}
            {interpMode=RCIM_Cubic, tangentMode=RCTM_User, time=0.5, value=1.2}
            {tangentMode=RCTM_User, time=1, value=0.4}
           ]}}}
        Scale Curve=&RandomRangeFloat(Minimum=0.7, Maximum=1)
       )
       B=&Lerp_Float(
        A=&Multiply_Float(A=$NPC.FortniteNPC.EventAudio_Amplitude, B=$NPC.FortniteNPC.EventAudio_Amplitude)
        B=1
        Alpha=$System.MusicReactivePreview
       )
      )
     )
    ))
   &Color(Color=&LerpLinearColors(StartColor=<rgba 0 1.2817 2 1>, EndColor=<rgba 0 1 0.4968 1>, LerpFactor=$Particles.Randomfloat))
   &ScaleColor(
    Color Value To Scale=$Particles.Color
    Scale Alpha=&Multiply_Float(
     A=&Multiply_Float(
      A=&Multiply_Float(A=1, B=&FN_FrontEndFloatModifier(Gameplay=1, Front End=0.8))
      B=&FNTimeOfDayScalar(Night Scale=0.7, Morning Scale=1, Evening Scale=1)
     )
     B=&FloatFromCurve(
      FloatCurve=DataInterface() {propertyValues {Curve {keys [
          {interpMode=RCIM_Cubic, tangentMode=RCTM_User, time=0, value=0}
          {interpMode=RCIM_Cubic, tangentMode=RCTM_User, time=0.5, value=1}
          {tangentMode=RCTM_User, time=1, value=0}
         ]}}}
      Scale Curve=&MakeCustomFloatFromBool(Bool=&SetBoolByFloatComparison(A=$NPC.FortniteNPC.EventAudio_Amplitude, B=0), True Float=1, False Float=0)
     )
    )
    Scale RGB=&Multiply_VectorByFloat(
     Vector=&MakeVectorFromLinearColorRGB(Color=&ColorFromCurve(ColorCurve=DataInterface() {propertyValues {
         AlphaCurve {keys []}
         BlueCurve {keys [
           <at 0 1 RCIM_Cubic>
           <at 0.1479 3.8798>
           {arriveTangent=-2.4295, interpMode=RCIM_Cubic, leaveTangent=-2.4295, time=0.3483, value=1.3807}
           <at 0.51 3>
          ]}
         GreenCurve {keys [
           <at 0 1 RCIM_Cubic>
           <at 0.1479 5>
           {arriveTangent=-9.785, interpMode=RCIM_Cubic, leaveTangent=-9.785, time=0.3483, value=0.0708}
           <at 0.51 1.4565>
          ]}
         RedCurve {keys [
           <at 0 1 RCIM_Cubic>
           <at 0.1479 1.8663>
           {arriveTangent=-4.7454, interpMode=RCIM_Cubic, leaveTangent=-4.7454, time=0.3483, value=5}
           <at 0.51 0.1478>
          ]}
        }}))
     Float=&RandomRangeFloat(Minimum=0.5, Maximum=&FN_FrontEndFloatModifier(Gameplay=1.2, Front End=0.6))
    )
    ScaleA=true
    ScaleRGB=true
   )
   /SetVariables(Particles.Color002=&Multiply_LinearColorByFloat(
     Linear Color=<rgba 0.0315 1 0.6961 0.2>
     Float=&Multiply_Float(
      A=&FNTimeOfDayScalar(Night Scale=0.8, Morning Scale=1, Evening Scale=1)
      B=&Multiply_Float(A=&FN_FrontEndFloatModifier(Gameplay=1, Front End=0.7), B=1)
     )
    )) {isSetParameters=true}
   /SetVariables(Particles.Size002:Vector3f=&Multiply_Vector(
     A=&MakeVector(X=0.8, Y=0.8, Z=&RandomRangeFloat(Minimum=13, Maximum=16))
     B=&MakeVector(
      X=1
      Y=1
      Z=&Multiply_Float(
       A=&FloatFromCurve(
        FloatCurve=DataInterface() {propertyValues {Curve {keys [
            {interpMode=RCIM_Cubic, tangentMode=RCTM_User, time=0, value=0.4}
            {interpMode=RCIM_Cubic, tangentMode=RCTM_User, time=0.5, value=1.2}
            {tangentMode=RCTM_User, time=1, value=0.4}
           ]}}}
        Scale Curve=&RandomRangeFloat(Minimum=0.7, Maximum=1)
       )
       B=&Lerp_Float(
        A=&Multiply_Float(A=$NPC.FortniteNPC.EventAudio_Amplitude, B=$NPC.FortniteNPC.EventAudio_Amplitude)
        B=1
        Alpha=$System.MusicReactivePreview
       )
      )
     )
    )) {isSetParameters=true}
   &DynamicMaterialParameters(
    Index 0 Param 1=&MakeFloatFromVector(Vector=&Multiply_Vector(A=<1 1 1>, B=$Particles.Scale), Channel=::NewEnumerator0)
    Index 0 Param 2=&MakeFloatFromVector(Vector=&Multiply_Vector(A=<1 1 1>, B=$Particles.Scale), Channel=::NewEnumerator1)
    Index 0 Param 3=&MakeFloatFromVector(Vector=&Multiply_Vector(A=<1 1 1>, B=$Particles.Scale), Channel=::NewEnumerator2)
    Index 1 Param 1=&Lerp_Float(A=-0.5, B=1, Alpha=$Particles.Randomfloat)
    Module.Index0Param0Write=true
    Module.Index0Param1Write=true
    Module.Index0Param2Write=true
    Module.Index0Param3Write=true
    Module.Index1Param0Write=true
    Module.Index1Param1Write=true
    Module.Index1Param2Write=true
    Module.Index1Param3Write=true
    Write Parameter Index 1=true
   )
  ]
  rendererClasses [&NiagaraMeshRendererProperties]
  renderers [
   {rendererClass=&NiagaraMeshRendererProperties, rendererIndex=0}
   {rendererClass=&NiagaraMeshRendererProperties, rendererIndex=1}
  ]
 }
 FN_Debug {
  bEnabled=false
  emitterUpdateScript [
   &EmitterState()
   &SpawnBurst_Instantaneous(Spawn Count=1, Use Spawn Probability=false)
  ]
  particleSpawnScript [/Niagara/Modules/Spawn/Initialization/InitializeParticle(
    Lifetime=5
    Module.Write Color=true
    Module.Write Lifetime=true
    Module.Write Mass=true
    Module.Write Material Random=true
    Module.Write Position=false
    Module.Write Scale=false
    Module.Write SpriteRotation=false
    Module.Write SpriteSize=true
    Position=/Niagara/Functions/Localspace/SimulationPosition()
    Sprite Size=<100 100>
   )]
  particleUpdateScript [
   &ParticleState()
   /SetVariables() {isSetParameters=true}
   &DynamicMaterialParameters(
    Index 0 Param 1=$Particles.Debug
    Module.Param0WriteEnabled=true
    Module.Param1WriteEnabled=true
    Module.Param2WriteEnabled=true
    Module.Param3WriteEnabled=true
   )
  ]
  rendererClasses [&NiagaraSpriteRendererProperties]
  renderers [{rendererClass=&NiagaraSpriteRendererProperties}]
 }
 Visualizer {
  bEnabled=false
  emitterUpdateScript [
   &EmitterState(
    Inactive Response=::NewEnumerator1
    Life Cycle Mode=::NewEnumerator1
    Loop Behavior=::NewEnumerator1
    Loop Duration Mode=::NewEnumerator1
    UseLoopDelay=false
   )
   /SetVariables(
    Emitter.OriginOffset:Vector3f=&TransformVector(Vector=<0 30 0>, Destination Space=::World)
    Emitter.PaddingPerCell=6
    Emitter.VolumeRadius=150
   ) {isSetParameters=true, module=SetVariables_#1}
   &SpawnParticlesInGrid(X Count=12, Y Count=12, Z Count=0)
  ]
  particleSpawnScript [
   &PartitionParticles(Number Of Partitions=3, Partition Mode:ENiagara_PartitionMode=::NewEnumerator5) {module=PartitionParticles001}
   &self:PartitionByBuildingSize_03(
    Fraction_MediumB=1
    Franction_TallB=1
    LBuilding_IndexMax=1
    LBuilding_IndexMin=1
    MBuilding_IndexMax=1
    MBuilding_IndexMin=1
    SBuilding_IndexMax=1
    SBuilding_IndexMin=1
   ) {module=PartitionByBuildingSize_03}
   &FN_SetSpawnGroup(Number of Spawn Groups=6) {module=FN_SetSpawnGroup001}
   /Niagara/Modules/Spawn/Initialization/V2/InitializeParticle(
    Mesh Scale Mode=::NewEnumerator3
    Mesh Uniform Scale=0.08
    Position Offset Coordinate Space=::Local
    UsePositionOffset=true
    Write Material Random=true
   )
   &GridLocation(
    Coordinate Space:ENiagaraCoordinateSpace=::World
    Module.Jitter Inside Grid Cells=false
    Module.Override Random Seed=false
    Normalize Offsets=false
    XYZ Dimensions:Vector3f=&MakeVector(
     X=&Divide_Float(A=$Emitter.VolumeRadius, B=$Emitter.PaddingPerCell)
     Y=&Divide_Float(A=$Emitter.VolumeRadius, B=$Emitter.PaddingPerCell)
     Z=0
    )
   )
   &InitialMeshOrientation(
    Mesh Orientation Mode=::NewEnumerator2
    Rotation=&SelectVectorFromArray(
     Array Sampling Mode=::NewEnumerator0
     Direct Array Index=?
     Vector Selection Array=DataInterface() {propertyValues {FloatData [<0 0 0>, <0 0 0.25>, <0 0 0.5>]}}
     Random Seed=0
    )
    Use Rotation Vector=true
   )
  ]
  particleUpdateScript [
   &SkeletalMeshLocation(
    MaskSpawnGroup=false
    Orientation Sampling=::NewEnumerator1
    OverrideRandomSeed=false
    Position Sampling=::NewEnumerator2
    Skeletal Mesh=DataInterface() {propertyValues {
      DefaultMesh=null
      ExcludeBoneName=root
      OptionalComponentClass=null
      PreviewMesh=null
      SkinningMode=SkinOnTheFly
      SoftSourceActor=null
      SourceMode=AttachParent
      UvSetIndex=0
      WholeMeshLOD=-1
      bExcludeBone=false
      bReadDeformedGeometry=true
      bRequireCurrentFrameData=true
      ComponentTags []
      FilteredBones []
      FilteredSockets []
      MeshUserParameter {parameter {name=None, typeDefHandle {registeredTypeIndex=1}, varData []}}
      SamplingRegions []
     }}
   )
   &ParticleState(Kill Particles When Lifetime Has Elapsed=false, Let Infinitely Lived Particles Die When Emitter Deactivates=true)
   &SocketLocation(
    Orientation Sampling=::NewEnumerator1
    Position Sampling=::NewEnumerator1
    SamplePosOffset=false
    Socket Reader=$System.Socket Reader
    Socket Sampling Mode=::NewEnumerator3
    VelocitySpeedLimit=false
   )
   &SolveForcesAndVelocity(Clamp Velocity=false, Limit Acceleration=false)
   /SetVariables(Particles.PositionCenter:NiagaraPosition=&AddVectorToPosition(Position=$Engine.Owner.Position, Vector=$Emitter.OriginOffset)) {isSetParameters=true, module=SetVariables_#2}
   &self:TreadmillParticles_01(
    GridSize:NiagaraFloat=&Multiply_Float(A=$Emitter.VolumeRadius, B=2)
    InputPosition:NiagaraPosition=$Particles.PositionCenter
   ) {module=TreadmillParticles_01}
   /SetVariables(Particles.Position=&MakeVector(
     X=&MakeFloatFromVector(Vector=Linked() {linkedVariable {name=Particles.Position, type=NiagaraPosition}}, Channel=::NewEnumerator0)
     Y=&MakeFloatFromVector(Vector=Linked() {linkedVariable {name=Particles.Position, type=NiagaraPosition}}, Channel=::NewEnumerator1)
     Z=&MakeFloatFromVector(Vector=/Niagara/DynamicInputs/Helpers/SimulationPosition(), Channel=::NewEnumerator2)
    )) {isSetParameters=true, module=SetVariables_#3}
   /SetVariables(Particles.Position=&AddVectorToPosition(Position=$Particles.Position, Vector=<0 0 100>)) {isSetParameters=true, module=SetVariables_#4}
   /SetVariables(Particles.Position=&AddVectorToPosition(
     Position=$Particles.Position
     Vector=&FN_FrontEndVector3Modifier(
      Gameplay=<0 0 0>
      Front End=&RotateVectorByQuaternion(Vector To Rotate=<0 -1.2 0>, Quaternion=$Output.SocketLocation.SampledRotation)
     )
    )) {isSetParameters=true, module=SetVariables_#5}
   &self:GridCollision_01(TraceDir:Vector3f=<0 0 -1>, TraceDist=250, ZBoost=250, bIsFrontEnd:NiagaraBool=$User.bIsFrontEndPreview) {module=GridCollision_01}
   &RegionMask(
    Invert Region=true
    Offset:Vector3f=$Emitter.OriginOffset
    Region Falloff=0
    Region Radius:NiagaraFloat=&Add_Float(
     A=&Multiply_Float(A=$Emitter.VolumeRadius, B=2)
     B=&Multiply_Float(A=$Emitter.VolumeRadius, B=0)
    )
    UseOffset=true
    UseRotation=false
   ) {enabled=false}
   &Timeline(
    Additional Output:ENiagara_TimelineOutput=::NewEnumerator1
    Additional Output Mode:ENiagara_TimelineOutputMode=::NewEnumerator1
    Float Curve=DataInterface() {propertyValues {Curve {keys [
        {arriveTangent=0.9218, interpMode=RCIM_Cubic, leaveTangent=0.9218, tangentMode=RCTM_User, time=0.35, value=0}
        {arriveTangent=2.875, interpMode=RCIM_Cubic, leaveTangent=2.875, time=0.63, value=0.9}
        {arriveTangent=-0.0156, interpMode=RCIM_Cubic, leaveTangent=-0.0156, tangentMode=RCTM_User, time=0.75, value=1.15}
        {arriveTangent=-0.0297, interpMode=RCIM_Cubic, leaveTangent=-0.0297, tangentMode=RCTM_User, time=0.9, value=0.95}
        {arriveTangent=0.0389, interpMode=RCIM_Cubic, leaveTangent=0.0389, tangentMode=RCTM_User, time=0.96, value=1.025}
        <at 1 1 RCIM_Cubic>
       ]}}}
    Play=&Boolean_NotOperation(Bool=$Particles.RegionMask.IsInsideMask)
    Play Rate=&MakeCustomFloatFromBool(
     Bool=&SetBoolByFloatComparison(A=&VectorLength(Vector=$User.OwnerLocalVelocity), B=160)
     True Float=10
     False Float=1
    )
    Rewind Rate:NiagaraFloat=&MakeCustomFloatFromBool(
     Bool=&SetBoolByFloatComparison(A=&VectorLength(Vector=$User.OwnerLocalVelocity), B=160)
     True Float=10
     False Float=3
    )
   )
   &ScaleMeshSize(Scale Factor=&VectorFromFloat(Value=&Multiply_Float(
      A=$Particles.Timeline.Float
      B=&FloatFromCurve(FloatCurve=DataInterface() {propertyValues {Curve {
          defaultValue=1
          keys [{interpMode=RCIM_Cubic, tangentMode=RCTM_User, time=0, value=1}, {interpMode=RCIM_Cubic, tangentMode=RCTM_User, time=1, value=1}]
         }}})
     ))) {module=ScaleMeshSize001}
   &ScaleMeshSize(Scale Factor=&AddVectorToPosition(
     Position=<1 1 1>
     Vector=&Multiply_Vector(
      A=<0 0 1>
      B=&VectorFromFloat(Value=&Multiply_Float(
        A=&VectorLength(Vector=&SubtractVectorFromPosition(Position=$Particles.SkeletalMeshLocation.SampledPosition, Vector=Linked() {linkedVariable {name=Particles.Position, type=NiagaraPosition}}))
        B=0.07
       ))
     )
    )) {module=ScaleMeshSize002}
   &ScaleMeshSize(Scale Factor=&Lerp_Vector(
     A=<1 1 1>
     B=<1 1 20>
     Alpha=&RemapRange(Input Value To Remap=$NPC.FortniteNPC.EventAudio_Amplitude, Input Min=0, Input Max=0.1, Output Min=0, Output Max=1)
    )) {enabled=false}
  ]
  rendererClasses [&NiagaraMeshRendererProperties]
  renderers [{rendererClass=&NiagaraMeshRendererProperties, rendererIndex=0}]
 }
}
summary {
 systemName=NS_FareSporeCookie_Test
 emitters [
  {bEnabled=false, emitterName=Visualizer, rendererClasses [&NiagaraMeshRendererProperties]}
  {bEnabled=false, emitterName=Box, rendererClasses [&NiagaraMeshRendererProperties]}
  {bEnabled=false, emitterName=BlowingParticles, simTarget=GPUComputeSim, rendererClasses [&NiagaraSpriteRendererProperties]}
  {emitterName=Audio_Visualizer, rendererClasses [&NiagaraMeshRendererProperties]}
  {bEnabled=false, emitterName=FN_Debug, rendererClasses [&NiagaraSpriteRendererProperties]}
 ]
 userVariables [
  User.bIsFrontEnd {type=NiagaraBool}
  User.bIsFrontEndPreview {type=NiagaraBool}
  User.OwnerLocalVelocity {type=Vector3f, description="", defaultValue {struct=Vector3f, value=<1 1 1>}}
 ]
}
